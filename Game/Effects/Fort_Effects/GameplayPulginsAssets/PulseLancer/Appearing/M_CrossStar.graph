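# Additive, unlit, two-sided Niagara material that draws a procedural cross star glint, filed under the PulseLancer appearing effects.
# A ProceduralCrossStar function shapes the star on center-scaled UVs, with tightness driven per particle by a dynamic parameter.
# A contrast-adjusted star mask blends a desaturated, tinted rainbow texture toward particle color for emissive.
# The refined star is maxed with a dimmer, powered second star, then multiplied by particle alpha.
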Material /Game/Effects/Fort_Effects/GameplayPulginsAssets/PulseLancer/Appearing/M_CrossStar
graph {
 properties {
  bAutomaticallySetUsageInEditor=false
  bUsedWithNiagaraMeshParticles=true
  bUsedWithNiagaraRibbons=true
  bUsedWithNiagaraSprites=true
  blendMode=BLEND_Additive
  shadingModel=MSM_Unlit
  twoSided=true
  usedShadingModels=Unlit
 }
 nodes [
  n1 = TextureObjectParameter() {group=Rainbow, parameterName=Rainbow_Texture, sortPriority=0, texture=&T_Tile_RainbowColor}
  n2 = VectorParameter() {group=Rainbow, parameterName=Rainbow_UVControl, channelNames {a=VPan, b=UPan, g=VScale, r=UScale}, defaultValue=<rgba 0.5 0.5 0.2 0.2>}
  n3 = Desaturation(
   None=Normalize(VectorInput=MaterialFunctionCall(Texture=@n1, UVControl_RGB=@n2, UVControl_A=@n2) {materialFunction=&FXCommon_TextureSample}.RGB)
   Fraction=ScalarParameter() {defaultValue=0.3, parameterName=DES}
  )
  n4 = VectorParameter() {parameterName=C2, defaultValue=<rgba 0.873 0.873 2 1>}
  n5 = ParticleColor()
  n6 = ScalarParameter() {defaultValue=1, parameterName=ssS}
  n7 = MaterialFunctionCall!(
   Tighten=DynamicParameter() {defaultValue=<rgba 2.3388 1 1 1>, paramNames [Tight,Param2,Param3,Param4]}.Tight
   UVIn=MaterialFunctionCall(UVs=TextureCoordinate(), Texture Scale=ScalarParameter() {defaultValue=0.8, parameterName=Sca}) {materialFunction=&ScaleUVsByCenter}.UVs
   "Softness(0-1)"=@n6
   "Fix_AntiFlick(Increase This If The Tip Flicking)"
   Fix_TipWidth=ScalarParameter() {defaultValue=1.5, parameterName=t}
  ) {materialFunction=/Game/Effects/Niagara/MasterMaterials/Functions/ProceduralCrossStar}
  n8 = Add(A=@n7.Result, B=ScalarParameter() {defaultValue=-0.5, group=ToC, parameterName=ToC_GrayAdjust_Add})
  n9 = ScalarParameter() {defaultValue=10, group=ToC, parameterName=ToC_GrayAdjust_Mul}
  n10 = Power(Base=Saturate(None=Multiply(A=@n8, B=@n9)), Exp=ScalarParameter() {defaultValue=6.2313, group=ToC, parameterName=ToC_GrayAdjust_Power})
  MaterialFunctionCall(Emissive=LinearInterpolate(A=Multiply(A=@n3, B=@n4.RGB), B=@n5.RGB, Alpha=Saturate(None=@n10))) {materialFunction=&MF_FX_EmissiveAdjustments}
  n11 = MaterialFunctionCall(Tighten=ScalarParameter() {defaultValue=1, parameterName=s}, "Softness(0-1)"=@n6) {materialFunction=&ProceduralCrossStar}
  n12 = ScalarParameter() {defaultValue=3, parameterName=fe2R}
  MaterialFunctionCall(In=@n11.LinearGradient0-1, Power=@n12) {materialFunction=&SCurve}
  n13 = VectorParameter() {parameterName=Star_Refine, defaultValue=<rgba 1.6074 1.3208 1 0>}
  n14 = Power(Base=@n11.Result, Exp=@n12)
  n15 = ScalarParameter() {defaultValue=0.3, parameterName=H}
  Multiply(
   A=Max(
    A=MaterialFunctionCall(Texture=@n7.Result, RGB=@n13.RGB, A=@n13.RGB) {materialFunction=&MF_FX_Refine}.Result
    B=Multiply(A=@n14, B=@n15) {constB=0.2}
   )
   B=@n5.A
  )
 ]
}
