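# M_Helix is an unlit, translucent, two-sided Niagara material for helix streaks in the PulseLancer appearing effect.
# Color multiplies a very bright tint by particle color, then hue-shifts it with cell noise for emissive adjustment.
# A mask combines refined noise, a U line profile, V tip/tail fades, alpha and an optional growth wipe with tail, driven by a dynamic parameter.
# A SplineThicken function sets strand width.
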Material /Game/Effects/Fort_Effects/GameplayPulginsAssets/PulseLancer/Appearing/M_Helix
graph {
 properties {
  bAutomaticallySetUsageInEditor=false
  bEnableResponsiveAA=true
  bOutputTranslucentVelocity=true
  bUsedWithNiagaraMeshParticles=true
  bUsedWithNiagaraRibbons=true
  bUsedWithNiagaraSprites=true
  blendMode=BLEND_Translucent
  shadingModel=MSM_Unlit
  twoSided=true
  usedShadingModels=Unlit
 }
 nodes [
  n1 = ScalarParameter() {defaultValue=4.3867, parameterName=SC}
  MaterialFunctionCall(WidthBase=@n1, WidthTip=@n1) {materialFunction=&SplineThicken}
  n2 = TextureObjectParameter() {
   group=HueShiftTexture
   parameterName=HueShiftTexture_Texture
   samplerType=SAMPLERTYPE_Grayscale
   sortPriority=0
   texture=/Game/Weapons/FORT_Melee/Pickaxe_Cube_Ninja_Male/FX/T_CellNoise
  }
  n3 = VectorParameter() {
   group=HueShiftTexture
   parameterName=HueShiftTexture_UVControl
   channelNames {a=VPan, b=UPan, g=VScale, r=UScale}
   defaultValue=<rgba 0 8.6727 0 0>
  }
  n4 = Multiply(
   A=MaterialFunctionCall(Texture=@n2, UVControl_RGB=@n3, UVControl_A=@n3) {materialFunction=&FXCommon_TextureSample}.R
   B=ScalarParameter() {defaultValue=-0.0448, parameterName=HS}
  )
  n5 = VectorParameter() {parameterName=Color, defaultValue=<rgba 50000 50000 50000 1>}
  n6 = ParticleColor()
  n7 = Multiply(A=@n5.RGB, B=@n6.RGB)
  MaterialFunctionCall(Emissive=Divide(A=MaterialFunctionCall(Hue Shift Percentage=@n4, Texture=@n7) {materialFunction=&HueShift}.Result)) {materialFunction=&MF_FX_EmissiveAdjustments}
  n8 = VectorParameter() {group=SandT, parameterName=SandT_UVControl, channelNames {a=VPan, b=UPan, g=VScale, r=UScale}, defaultValue=<rgba 0 4.0486 0 0>}
  n9 = MaterialFunctionCall(
   Texture=TextureObjectParameter() {group=SandT, parameterName=SandT_Texture, sortPriority=0, texture=/Game/Global/Textures/Default/Noise/Noise}
   UVControl_RGB=@n8
   UVControl_A=@n8
  ) {materialFunction=&FXCommon_TextureSample}
  n10 = VectorParameter() {parameterName=SandT_Refine, defaultValue=<rgba 6.0309 11.2959 0.9454 0.5>}
  n11 = OneMinus(None=MaterialFunctionCall(Float2=TextureCoordinate()) {materialFunction=&BreakOutFloat2Components}.G)
  n12 = Add(
   A=DynamicParameter() {paramNames [Growth,Param2,Param3,Param4]}.Growth
   B=ScalarParameter() {defaultValue=1, group=Growth, parameterName=Growth_GrayAdjust_Add}
  )
  n13 = Add(A=@n11, B=Add(A=@n12, B=ScalarParameter() {defaultValue=-0.912, group=GrowthTail, parameterName=GrowthTail_GrayAdjust_Add}))
  n14 = Saturate(None=Multiply(A=OneMinus(None=@n13), B=ScalarParameter() {defaultValue=1, group=GrowthTail, parameterName=GrowthTail_GrayAdjust_Mul}))
  n15 = ScalarParameter() {defaultValue=1, group=GrowthTail, parameterName=GrowthTail_GrayAdjust_Power}
  n16 = Saturate(None=Multiply(A=Add(A=@n11, B=@n12), B=ScalarParameter() {defaultValue=1, group=Growth, parameterName=Growth_GrayAdjust_Mul}))
  n17 = ScalarParameter() {defaultValue=1, group=Growth, parameterName=Growth_GrayAdjust_Power}
  n18 = Multiply(A=Saturate(None=Power(Base=@n14, Exp=@n15)), B=Saturate(None=Power(Base=@n16, Exp=@n17)))
  n19 = ScalarParameter() {parameterName=EnableGrowth}
  n20 = MaterialFunctionCall(Float2=TextureCoordinate()) {materialFunction=&BreakOutFloat2Components}
  n21 = VectorParameter() {
   group=Line
   parameterName=Line_EdgeMask_Control
   channelNames {a="", b=Power, g=Multiply, r=Offset}
   defaultValue=<rgba -0.3072 2.3049 9.1001 1>
  }
  n22 = Multiply(
   A=@n5.A
   B=Multiply(
    A=MaterialFunctionCall!(UorV=@n20.R, EdgeControl=@n21) {materialFunction=/Game/Effects/Niagara/MasterMaterials/Functions/Common/FXCommon_EdgeMask_Simple}.Mask
    B=@n6.A
   )
  )
  n23 = VectorParameter() {
   group=TipTailFade
   parameterName=TipTailFade_EdgeMask_InvertControl
   channelNames {a="", b=Power, g=Multiply, r=Offset}
   defaultValue=<rgba 0 5.3272 2 1>
  }
  n24 = VectorParameter() {
   group=TipTailFade
   parameterName=TipTailFade_EdgeMask_Control
   channelNames {a="", b=Power, g=Multiply, r=Offset}
   defaultValue=<rgba 0 10 2 1>
  }
  n25 = MaterialFunctionCall!(UorV=@n20.G, Invert_EdgeControl=@n23, NoInvert_EdgeControl=@n24) {materialFunction=/Game/Effects/Niagara/MasterMaterials/Functions/Common/FXCommon_EdgeMask_Ex}
  Multiply(
   A=LinearInterpolate(
    B=MaterialFunctionCall(Texture=@n9.R, RGB=@n10.RGB, A=@n10.RGB) {materialFunction=&MF_FX_Refine}.Result
    Alpha=ScalarParameter() {defaultValue=1, parameterName=EnableSandT}
   ) {constA=1}
   B=Multiply(A=LinearInterpolate(B=@n18, Alpha=@n19) {constA=1}, B=Multiply(A=@n22, B=@n25.Mask)) {constA=1}
  )
  TextureCoordinate()
 ]
}
