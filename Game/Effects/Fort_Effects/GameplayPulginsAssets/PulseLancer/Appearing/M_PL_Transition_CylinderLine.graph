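# Additive, unlit, two-sided Niagara material that draws glowing, noisy line streaks on a cylinder for the Pulse Lancer appearing transition.
# Bounds-normalized position, rotated by a dynamic parameter, feeds a noise-warped sine with inverse-proportional decay for thin lines.
# Two panning Perlin layers break up the lines, while Appear and Dissolve parameters shape masks.
# Scaled normals drive gradient- and noise-based vertex offset; face-dependent colors multiply particle color.
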Material /Game/Effects/Fort_Effects/GameplayPulginsAssets/PulseLancer/Appearing/M_PL_Transition_CylinderLine
graph {
 properties {
  bAutomaticallySetUsageInEditor=false
  bUsedWithNiagaraMeshParticles=true
  bUsedWithNiagaraRibbons=true
  bUsedWithNiagaraSprites=true
  blendMode=BLEND_Additive
  shadingModel=MSM_Unlit
  twoSided=true
  usedShadingModels=Unlit
 }
 nodes [
  n1 = VertexNormalWS()
  n2 = Transform(None=Multiply(
    A=Transform(None=@n1) {transformSourceType=TRANSFORMSOURCE_World, transformType=TRANSFORM_Local}
    B=VectorParameter() {parameterName=ScaleNormal, defaultValue=<rgba 0.5 1 1 1>}.RGB
   )) {transformSourceType=TRANSFORMSOURCE_Local}
  n3 = ScalarParameter() {defaultValue=-0.3706, parameterName=test3y2}
  n4 = ScalarParameter() {defaultValue=15.9045, parameterName=test2}
  n5 = Normalize(VectorInput=VectorParameter() {parameterName=RA, defaultValue=<rgba 1 0 0 0.0256>}.RGB)
  n6 = DynamicParameter() {paramNames [Appear,Dissolve,Rotate,WPO]}
  n7 = Constant3Vector()
  n8 = WorldPosition() {worldPositionShaderOffset=WPT_ExcludeAllShaderOffsets}
  n9 = Custom!(None) {code="GetPrimitiveData(Parameters).InstanceLocalBoundsCenter.xyz", description=BoundsCenter}
  n10 = Custom() {code="GetPrimitiveData(Parameters).InstanceLocalBoundsExtent.xyz", description=BoundsExtent}
  n11 = Divide(
   A=Subtract(
    A=TransformPosition(None=@n8.XYZ) {transformSourceType=TRANSFORMPOSSOURCE_World, transformType=TRANSFORMPOSSOURCE_Instance}
    B=Subtract(A=@n9, B=@n10)
   )
   B=Multiply(A=@n10) {constB=2}
  )
  n12 = MaterialFunctionCall(Float3=Add(A=RotateAboutAxis(NormalizedRotationAxis=@n5, RotationAngle=@n6.Rotate, PivotPoint=@n7, Position=@n11), B=@n11)) {materialFunction=&BreakOutFloat3Components}
  n13 = OneMinus(None=@n12.B)
  n14 = DynamicParameter() {paramNames [Appear,Dissolve,Rotate,WPO]}
  n15 = Saturate(None=Multiply(A=Add(A=@n13, B=@n14.Appear), B=ScalarParameter() {defaultValue=2.0085, group=VertexMask, parameterName=VertexMask_GrayAdjust_Mul}))
  n16 = Saturate(None=Power(Base=OneMinus(None=@n15), Exp=ScalarParameter() {defaultValue=5, group=VertexMask, parameterName=VertexMask_GrayAdjust_Power}))
  n17 = DynamicParameter() {paramNames [Appear,Dissolve,Rotate,WPO]}
  n18 = TextureObjectParameter() {
   group=WPOT
   parameterName=WPOT_Texture
   samplerType=SAMPLERTYPE_LinearColor
   sortPriority=0
   texture=/Game/Weapons/FORT_Melee/Pickaxe_RhombCamo/FX/T_Pickaxe_RhombCamo_Perlin
  }
  n19 = VectorParameter() {group=WPOT, parameterName=WPOT_UVControl, channelNames {a=VPan, b=UPan, g=VScale, r=UScale}, defaultValue=<rgba 0 5 0 0>}
  n20 = MaterialFunctionCall() {materialFunction=/Game/Effects/Niagara/MasterMaterials/Functions/InstanceBounds_0-1_UVW}
  n21 = AppendVector(A=@n20.G, B=@n12.B)
  n22 = Multiply(
   A=Add(
    A=MaterialFunctionCall(Texture=@n18, UVControl_RGB=@n19, UVControl_A=@n19, "[Override]UVInput"=@n21) {materialFunction=&FXCommon_TextureSample}.R
    B=ScalarParameter() {group=NoiseToWPO, parameterName=NoiseToWPO_GrayAdjust_Add}
   )
   B=ScalarParameter() {defaultValue=1, group=NoiseToWPO, parameterName=NoiseToWPO_GrayAdjust_Mul}
  )
  n23 = Saturate(None=Power(Base=Saturate(None=@n22), Exp=ScalarParameter() {defaultValue=2.0204, group=NoiseToWPO, parameterName=NoiseToWPO_GrayAdjust_Power}))
  Add(A=Multiply(A=@n2, B=LinearInterpolate(A=@n3, B=@n4, Alpha=@n16)), B=Multiply(A=@n2, B=Multiply(A=@n17.WPO, B=@n23)))
  n24 = LinearInterpolate(
   A=VectorParameter() {parameterName=LineCB, defaultValue=<rgba 2.2245 1.04 5 1>}.RGB
   B=VectorParameter() {parameterName=LineC, defaultValue=<rgba 6.9283 5.6174 10 1>}.RGB
   Alpha=Saturate(None=TwoSidedSign())
  )
  n25 = ParticleColor()
  MaterialFunctionCall(Emissive=Divide(A=Multiply(A=@n24, B=@n25.RGB))) {materialFunction=&MF_FX_EmissiveAdjustments}
  n26 = MaterialFunctionCall(Value=Add(A=@n12.B, B=NamedRerouteUsage()), Period=ScalarParameter() {defaultValue=0.015, parameterName=P}, -1 to 1=StaticBool()) {materialFunction=&LinearSine}
  n27 = VectorParameter() {parameterName=Line_Refine, defaultValue=<rgba 8.072 1 1.5 0.5>}
  n28 = MaterialFunctionCall(
   Texture=MaterialFunctionCall!(
    Emissive=ScalarParameter() {defaultValue=182.1822, parameterName=ThirdShape_IPDE}
    Gradient=NamedRerouteDeclaration(None=@n26.Linear Sine) {name=SineLinear, nodeColor=<rgba 1 0.0234 0 1>}
   ) {materialFunction=/Game/Effects/Fort_Effects/GameplayPulginsAssets/PulseLancer/Slash/InverseProportionalDecay}.Result
   RGB=@n27.RGB
   A=@n27.RGB
  ) {materialFunction=&MF_FX_Refine}
  n29 = TextureObjectParameter() {
   group=TheWireTexture
   parameterName=TheWireTexture_Texture
   samplerType=SAMPLERTYPE_LinearColor
   sortPriority=0
   texture=&T_Pickaxe_RhombCamo_Perlin
  }
  n30 = VectorParameter() {
   group=TheWireTexture
   parameterName=TheWireTexture_UVControl
   channelNames {a=VPan, b=UPan, g=VScale, r=UScale}
   defaultValue=<rgba 0.3 10 1.6038 -0.0796>
  }
  n31 = MaterialFunctionCall(Texture=@n29, UVControl_RGB=@n30, UVControl_A=@n30, "[Override]UVInput"=@n21) {materialFunction=&FXCommon_TextureSample}
  n32 = ScalarParameter() {group=TheWireTexture, parameterName=TheWireTexture_GrayAdjust_Add}
  n33 = Saturate(None=Multiply(A=Add(A=@n31.R, B=@n32), B=ScalarParameter() {defaultValue=1.2, group=TheWireTexture, parameterName=TheWireTexture_GrayAdjust_Mul}))
  n34 = Add(A=@n14.Dissolve, B=ScalarParameter() {defaultValue=-0.0896, group=DissolveMask, parameterName=DissolveMask_GrayAdjust_Add})
  n35 = Saturate(None=Multiply(A=Add(A=@n13, B=@n34), B=ScalarParameter() {defaultValue=1, group=DissolveMask, parameterName=DissolveMask_GrayAdjust_Mul}))
  n36 = ScalarParameter() {defaultValue=1, group=DissolveMask, parameterName=DissolveMask_GrayAdjust_Power}
  n37 = LinearInterpolate(
   A=ScalarParameter() {group=TheWireTexture, parameterName=TheWireTexture_GrayAdjust_Power}
   B=ScalarParameter() {defaultValue=12.2974, group=TheWireTexture, parameterName=TheWireTexture_GrayAdjust_Power2}
   Alpha=Saturate(None=Power(Base=@n35, Exp=@n36))
  )
  n38 = TextureObjectParameter() {
   group=Wiretexture2
   parameterName=Wiretexture2_Texture
   samplerType=SAMPLERTYPE_LinearGrayscale
   sortPriority=0
   texture=&T_SimplePerlin_Greyscale
  }
  n39 = VectorParameter() {
   group=Wiretexture2
   parameterName=Wiretexture2_UVControl
   channelNames {a=VPan, b=UPan, g=VScale, r=UScale}
   defaultValue=<rgba 0.1 4 7.7997 0>
  }
  n40 = MaterialFunctionCall(Texture=@n38, UVControl_RGB=@n39, UVControl_A=@n39, "[Override]UVInput"=@n21) {materialFunction=&FXCommon_TextureSample}
  n41 = ScalarParameter() {defaultValue=0.1332, group=Wiretexture2, parameterName=Wiretexture2_GrayAdjust_Add}
  n42 = Saturate(None=Multiply(A=Add(A=@n40.R, B=@n41), B=ScalarParameter() {defaultValue=1.5809, group=Wiretexture2, parameterName=Wiretexture2_GrayAdjust_Mul}))
  n43 = ScalarParameter() {defaultValue=6.1149, group=Wiretexture2, parameterName=Wiretexture2_GrayAdjust_Power}
  n44 = Multiply(A=Saturate(None=Power(Base=@n33, Exp=@n37)), B=Saturate(None=Power(Base=@n42, Exp=@n43)))
  n45 = Add(A=@n13, B=@n14.Appear)
  n46 = ScalarParameter() {defaultValue=3.1324, group=MaskAppear, parameterName=MaskAppear_GrayAdjust_Mul}
  n47 = Power(
   Base=Saturate(None=Multiply(A=@n45, B=@n46))
   Exp=ScalarParameter() {defaultValue=2, group=MaskAppear, parameterName=MaskAppear_GrayAdjust_Power}
  )
  Multiply(A=Multiply(A=Saturate(None=@n47), B=Multiply(A=@n28.Result, B=@n44)), B=@n25.A)
  n48 = TextureObjectParameter() {
   group=Noise
   parameterName=Noise_Texture
   samplerType=SAMPLERTYPE_LinearColor
   sortPriority=0
   texture=&T_Pickaxe_RhombCamo_Perlin
  }
  n49 = VectorParameter() {group=Noise, parameterName=Noise_UVControl, channelNames {a=VPan, b=UPan, g=VScale, r=UScale}, defaultValue=<rgba 0.5 5.538 1 -0.1>}
  n50 = VectorParameter() {
   group=Noise
   parameterName=Noise_NoiseControl
   sortPriority=60
   channelNames {a=VStrength, b=UStrength, g=VOffset, r=UOffset}
   defaultValue=<rgba -0.5 -0.5 0.01 0>
  }
  NamedRerouteDeclaration(None=MaterialFunctionCall!(
    NoiseValue=MaterialFunctionCall(Texture=@n48, UVControl_RGB=@n49, UVControl_A=@n49, "[Override]UVInput"=@n21) {materialFunction=&FXCommon_TextureSample}.R
    Control_RGB=@n50
    Control_A=@n50
   ) {materialFunction=/Game/Effects/Niagara/MasterMaterials/Functions/Common/FXCommon_NoiseControl}.NoiseValue) {name=NoiseUV, nodeColor=<rgba 0 0.8359 1 1>}
  OneMinus(None=@n35)
  ScalarParameter() {parameterName=Test}
  ScalarParameter() {defaultValue=0.75, group=MaskAppear, parameterName=MaskAppear_GrayAdjust_Add}
  n51 = Multiply(A=@n20.R) {constB=0.5}
  n52 = OneMinus(None=@n51)
  n53 = Floor(None=Multiply(A=@n20.G) {constB=2})
  n54 = Constant3Vector() {constant=<rgba 1 0 0 0>}
  n55 = Add(
   A=RotateAboutAxis(
    NormalizedRotationAxis=Constant3Vector() {constant=<rgba 0 0 1 0>}
    RotationAngle=LinearInterpolate(A=@n51, B=@n52, Alpha=@n53)
    PivotPoint=Constant3Vector()
    Position=@n54
   )
   B=@n54
  )
  n56 = RotateAboutAxis(NormalizedRotationAxis=@n5, RotationAngle=Multiply(A=@n6.Rotate) {constB=-1}, PivotPoint=Constant3Vector(), Position=@n55)
  Transform(None=Normalize(VectorInput=Add(A=@n56, B=@n55))) {transformSourceType=TRANSFORMSOURCE_Instance}
  VertexNormalWS()
  VertexNormalWS()
 ]
}
