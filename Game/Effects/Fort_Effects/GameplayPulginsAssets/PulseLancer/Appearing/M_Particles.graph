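# Additive, unlit, two-sided Niagara particle material for the PulseLancer appearing effect, usable on sprites, ribbons and meshes.
# A tiling noise texture is refined two ways, tinted by particle color and a saturation-adjusted copy, then summed into emissive adjustments.
# A noise normal map distorts UVs for a circular SphereMask, multiplied by particle alpha and squared depth fade.
# The RDM dynamic parameter offsets both texture samples, allowing per-particle variation.
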Material /Game/Effects/Fort_Effects/GameplayPulginsAssets/PulseLancer/Appearing/M_Particles
graph {
 properties {
  bAutomaticallySetUsageInEditor=false
  bUsedWithNiagaraMeshParticles=true
  bUsedWithNiagaraRibbons=true
  bUsedWithNiagaraSprites=true
  blendMode=BLEND_Additive
  previewMesh=/Engine/Transient.MaterialEditorMeshComponent_9
  shadingModel=MSM_Unlit
  twoSided=true
  usedShadingModels=Unlit
 }
 nodes [
  n1 = ParticleColor()
  n2 = DynamicParameter() {defaultValue=<rgba -0.2608 1 1 1>, paramNames [RDM,Param2,Param3,Param4]}
  n3 = VectorParameter() {group=SandT, parameterName=SandT_UVControl, channelNames {a=VPan, b=UPan, g=VScale, r=UScale}, defaultValue=<rgba 0.5 0.5 0 0>}
  n4 = MaterialFunctionCall(
   Texture=TextureObjectParameter() {group=SandT, parameterName=SandT_Texture, sortPriority=0, texture=/Game/Global/Textures/Default/Noise/T_TileNoise_02}
   AdditionalUVOffset=@n2.RDM
   UVControl_RGB=@n3
   UVControl_A=@n3
  ) {materialFunction=&FXCommon_TextureSample}
  n5 = VectorParameter() {parameterName=SandT_Refine, defaultValue=<rgba 5.2298 5986.1851 1 0.5>}
  n6 = MaterialFunctionCall(Texture=@n4.R, RGB=@n5.RGB, A=@n5.RGB) {materialFunction=&MF_FX_Refine}
  n7 = Desaturation(None=@n1.RGB, Fraction=ScalarParameter() {defaultValue=-0.3, parameterName=des2})
  n8 = VectorParameter() {parameterName=SandT2_Refine, defaultValue=<rgba 0.5 0.5894 0 0>}
  n9 = MaterialFunctionCall(Texture=@n4.R, RGB=@n8.RGB, A=@n8.RGB) {materialFunction=&MF_FX_Refine}
  MaterialFunctionCall(Emissive=Add(A=Multiply(A=@n1.RGB, B=@n6.Result), B=Multiply(A=@n7, B=@n9.Result))) {materialFunction=&MF_FX_EmissiveAdjustments}
  n10 = TextureObjectParameter() {
   group=Mask
   parameterName=Mask_Texture
   samplerType=SAMPLERTYPE_Normal
   sortPriority=0
   texture=/Game/Characters/Player/Female/Medium/Bodies/F_MED_Bistro_Astronaut/FX/T_GaussianNoise_Normal
  }
  n11 = VectorParameter() {group=Mask, parameterName=Mask_UVControl, channelNames {a=VPan, b=UPan, g=VScale, r=UScale}, defaultValue=<rgba 0.125 0.125 0 0>}
  n12 = VectorParameter() {
   group=Mask
   parameterName=Mask_NoiseControl
   sortPriority=60
   channelNames {a=VStrength, b=UStrength, g=VOffset, r=UOffset}
   defaultValue=<rgba 0 0 0.25 0.25>
  }
  n13 = Add(
   A=TextureCoordinate()
   B=MaterialFunctionCall!(
    NoiseValue=MaterialFunctionCall(Texture=@n10, AdditionalUVOffset=@n2.RDM, UVControl_RGB=@n11, UVControl_A=@n11) {materialFunction=&FXCommon_TextureSample}.RG
    Control_RGB=@n12
    Control_A=@n12
   ) {materialFunction=/Game/Effects/Niagara/MasterMaterials/Functions/Common/FXCommon_NoiseControl}.NoiseValue
  )
  n14 = Constant2Vector() {g=0.5, r=0.5}
  n15 = ScalarParameter() {defaultValue=0.1, parameterName=Hardness}
  n16 = DepthFade(FadeDistance=ScalarParameter() {parameterName=DF})
  Multiply(A=Multiply(A=@n1.A, B=SphereMask(A=@n13, B=@n14, Hardness=@n15) {attenuationRadius=0.25, hardnessPercent=0}), B=Multiply(A=@n16, B=@n16))
 ]
}
