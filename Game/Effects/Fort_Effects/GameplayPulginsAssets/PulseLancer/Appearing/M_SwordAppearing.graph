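# Masked, lit Niagara particle material for the PulseLancer effects that, per its name, makes a sword appear.
# Adds a dynamic parameter to the instance-bounds Z gradient, driving a sweep along the mesh.
# A power of that gradient makes a bright HDR CColor band, added to particle color hue-shifted along the same axis.
# A second, inverted power forms the cutoff mask, with temporal AA dithering below a near-one threshold.
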Material /Game/Effects/Fort_Effects/GameplayPulginsAssets/PulseLancer/Appearing/M_SwordAppearing
graph {
 properties {
  bAutomaticallySetUsageInEditor=false
  bUsedWithNiagaraMeshParticles=true
  bUsedWithNiagaraRibbons=true
  bUsedWithNiagaraSprites=true
  blendMode=BLEND_Masked
  usedShadingModels=Default Lit
 }
 nodes [
  n1 = MaterialFunctionCall() {materialFunction=/Game/Effects/Niagara/MasterMaterials/Functions/InstanceBounds_0-1_UVW}
  n2 = Add(A=@n1.B, B=DynamicParameter() {defaultValue=<rgba 0 1 1 1>}.R)
  n3 = ScalarParameter() {defaultValue=15.3864, parameterName=eg1}
  n4 = ScalarParameter() {defaultValue=0.1793, parameterName=h}
  Add(
   A=Multiply(A=VectorParameter() {parameterName=CColor, defaultValue=<rgba 792.2596 373.9172 5000 1>}.RGB, B=Power(Base=@n2, Exp=@n3))
   B=MaterialFunctionCall(Hue Shift Percentage=Multiply(A=@n1.B, B=@n4), Texture=ParticleColor().RGB) {materialFunction=&HueShift}.Result
  )
  n5 = OneMinus(None=Power(Base=@n2, Exp=ScalarParameter() {defaultValue=18.6642, parameterName=eg2}))
  n6 = Constant() {r=1}
  If(
   A=@n5
   "A > B"=@n6
   "A == B"=@n6
   "A < B"=MaterialFunctionCall(Alpha Threshold=@n5) {materialFunction=&DitherTemporalAA_Fortnite}.Result
  ) {constB=0.99}
  VectorParameter() {parameterName=BaseColor, defaultValue=<rgba 0.3409 0.3409 0.3409 1>}
 ]
}
